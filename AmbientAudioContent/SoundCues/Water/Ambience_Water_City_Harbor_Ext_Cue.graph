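# Sound cue for an exterior city harbor water ambience, built as a looping bed of ocean wave sounds.
# A mixer layers three looping waves (small shore waves, medium-distant, distant) at separate volumes, with the distant layer loudest.
# The cue plays at reduced volume in the water ambient sound class and references shared ambient attenuation.
# It also stores override values for distance low-pass filtering and a single-instance, prevent-new concurrency limit.
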SoundCue /AmbientAudioContent/SoundCues/Water/Ambience_Water_City_Harbor_Ext_Cue
root=Mixer_2
nodes [
 {id=Mixer_2, children [WavePlayer_8,WavePlayer_6,WavePlayer_3], properties {inputVolume [0.6,0.8,1.25]}}
 {id=WavePlayer_3, properties {bLooping=true, soundWaveAssetPtr=/Game/Athena/Sounds/Water/Ambient/Amb_Water_Ocean_Waves_Distant_Loop_01}}
 {id=WavePlayer_6, properties {bLooping=true, soundWaveAssetPtr=/Game/Athena/Sounds/Water/Ambient/Amb_Water_Ocean_Waves_Med_Distant_Loop_01}}
 {id=WavePlayer_8, properties {bLooping=true, soundWaveAssetPtr=/Game/Athena/Sounds/Water/Ambient/Amb_Water_Ocean_Shore_SmWaves_Loop_01}}
]
properties {
 attenuationSettings=&Fort_Ambient_Athena_Att
 maxDistance=2097152
 soundClassObject=/Game/Sounds/FortSoundClassesAndModes/Fort_Water_Ambient
 volumeMultiplier=0.25
 attenuationOverrides {
  bAttenuateWithLPF=true
  falloffDistance=6000
  lPFFrequencyAtMax=500
  nonSpatializedRadiusStart=700
  stereoSpread=1000
  attenuationShapeExtents {x=925}
 }
 concurrencyOverrides {maxCount=1, resolutionRule=PreventNew}
}
