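# Ambient sound cue for a spot on a frozen river, playing randomized ice breaks, pings and shifts.
# Four looping layers (large breaks, small breaks, pings, frozen-lake shifts) each pair a random delay with pitch modulation and random wave selection, with breaks spaced furthest apart.
# A weighted mixer blends them through a close river attenuation node.
# The cue uses the water ambient sound class and overrides concurrency to one instance.
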SoundCue /AmbientAudioContent/SoundCues/Water/SC_Spot_River_Frozen_Ice_Pings_Breaks
root=Mixer_0
nodes [
 {
  id=Attenuation_3
  children [Mixer_1]
  properties {attenuationSettings=/AmbientAudioBase/Attenuations/Apollo_Water_River_Close_Gold_Focus_Att}
 }
 {id=Delay_0, children [Modulator_0], properties {delayMax=18, delayMin=12}}
 {id=Delay_1, children [Modulator_1], properties {delayMax=12, delayMin=6}}
 {id=Delay_2, children [Modulator_2], properties {delayMax=6, delayMin=2}}
 {id=Delay_3, children [Modulator_3], properties {delayMax=6, delayMin=2}}
 {id=Looping_0, children [Delay_0]}
 {id=Looping_1, children [Delay_1]}
 {id=Looping_2, children [Delay_2]}
 {id=Looping_3, children [Delay_3]}
 {id=Mixer_0, children [Attenuation_3, null], properties {inputVolume [1,1]}}
 {id=Mixer_1, children [Looping_0,Looping_1,Looping_2,Looping_3], properties {inputVolume [0.6,0.28,1,1]}}
 {id=Modulator_0, children [Random_1], properties {pitchMax=1.1, pitchMin=1, volumeMax=1.6, volumeMin=1.5}}
 {id=Modulator_1, children [Random_0], properties {pitchMax=1.1, pitchMin=1}}
 {id=Modulator_2, children [Random_2], properties {pitchMax=1.1, pitchMin=1, volumeMax=4, volumeMin=2}}
 {id=Modulator_3, children [Random_3], properties {pitchMax=1.1, pitchMin=1, volumeMax=1.5, volumeMin=1.25}}
 {
  id=Random_0
  children [
   WavePlayer_15
   WavePlayer_16
   WavePlayer_17
   WavePlayer_18
   WavePlayer_19
   WavePlayer_20
   WavePlayer_21
   WavePlayer_22
   WavePlayer_23
   WavePlayer_24
  ]
  properties {weights [1,1,1,1,1,1,1,1,1,1]}
 }
 {id=Random_1, children [WavePlayer_0,WavePlayer_1,WavePlayer_2,WavePlayer_3,WavePlayer_4], properties {weights [1,1,1,1,1]}}
 {id=Random_2, children [WavePlayer_25, WavePlayer_26, WavePlayer_27, WavePlayer_28, WavePlayer_29], properties {weights [1,1,1,1,1]}}
 {id=Random_3, children [WavePlayer_30, WavePlayer_31, WavePlayer_32, WavePlayer_33], properties {weights [1,1,1,1]}}
 {id=WavePlayer_0, properties {soundWaveAssetPtr=/AmbientAudioContent/Waves/SpottedLoops/Spot_IceCastle_Ice_Stalacitte_Break_01}}
 {id=WavePlayer_1, properties {soundWaveAssetPtr=/AmbientAudioContent/Waves/SpottedLoops/Spot_IceCastle_Ice_Stalacitte_Break_02}}
 {id=WavePlayer_15, properties {soundWaveAssetPtr=/AmbientAudioContent/Waves/SpottedLoops/Spot_IceCastle_Ice_Stalacitte_Break_Small_01}}
 {id=WavePlayer_16, properties {soundWaveAssetPtr=/AmbientAudioContent/Waves/SpottedLoops/Spot_IceCastle_Ice_Stalacitte_Break_Small_02}}
 {id=WavePlayer_17, properties {soundWaveAssetPtr=/AmbientAudioContent/Waves/SpottedLoops/Spot_IceCastle_Ice_Stalacitte_Break_Small_03}}
 {id=WavePlayer_18, properties {soundWaveAssetPtr=/AmbientAudioContent/Waves/SpottedLoops/Spot_IceCastle_Ice_Stalacitte_Break_Small_04}}
 {id=WavePlayer_19, properties {soundWaveAssetPtr=/AmbientAudioContent/Waves/SpottedLoops/Spot_IceCastle_Ice_Stalacitte_Break_Small_05}}
 {id=WavePlayer_2, properties {soundWaveAssetPtr=/AmbientAudioContent/Waves/SpottedLoops/Spot_IceCastle_Ice_Stalacitte_Break_03}}
 {id=WavePlayer_20, properties {soundWaveAssetPtr=/AmbientAudioContent/Waves/SpottedLoops/Spot_IceCastle_Ice_Stalacitte_Break_Small_06}}
 {id=WavePlayer_21, properties {soundWaveAssetPtr=/AmbientAudioContent/Waves/SpottedLoops/Spot_IceCastle_Ice_Stalacitte_Break_Small_07}}
 {id=WavePlayer_22, properties {soundWaveAssetPtr=/AmbientAudioContent/Waves/SpottedLoops/Spot_IceCastle_Ice_Stalacitte_Break_Small_08}}
 {id=WavePlayer_23, properties {soundWaveAssetPtr=/AmbientAudioContent/Waves/SpottedLoops/Spot_IceCastle_Ice_Stalacitte_Break_Small_09}}
 {id=WavePlayer_24, properties {soundWaveAssetPtr=/AmbientAudioContent/Waves/SpottedLoops/Spot_IceCastle_Ice_Stalacitte_Break_Small_10}}
 {id=WavePlayer_25, properties {soundWaveAssetPtr=/AmbientAudioContent/Waves/OneShots/Spot_RealityWarrior_Ice_Ping_01}}
 {id=WavePlayer_26, properties {soundWaveAssetPtr=/AmbientAudioContent/Waves/OneShots/Spot_RealityWarrior_Ice_Ping_02}}
 {id=WavePlayer_27, properties {soundWaveAssetPtr=/AmbientAudioContent/Waves/OneShots/Spot_RealityWarrior_Ice_Ping_03}}
 {id=WavePlayer_28, properties {soundWaveAssetPtr=/AmbientAudioContent/Waves/OneShots/Spot_RealityWarrior_Ice_Ping_04}}
 {id=WavePlayer_29, properties {soundWaveAssetPtr=/AmbientAudioContent/Waves/OneShots/Spot_RealityWarrior_Ice_Ping_05}}
 {id=WavePlayer_3, properties {soundWaveAssetPtr=/AmbientAudioContent/Waves/SpottedLoops/Spot_IceCastle_Ice_Stalacitte_Break_04}}
 {id=WavePlayer_30, properties {soundWaveAssetPtr=/AmbientAudioContent/Waves/Water/Spot_IceRealm_FrozenLake_IceShifts_01}}
 {id=WavePlayer_31, properties {soundWaveAssetPtr=/AmbientAudioContent/Waves/Water/Spot_IceRealm_FrozenLake_IceShifts_02}}
 {id=WavePlayer_32, properties {soundWaveAssetPtr=/AmbientAudioContent/Waves/Water/Spot_IceRealm_FrozenLake_IceShifts_03}}
 {id=WavePlayer_33, properties {soundWaveAssetPtr=/AmbientAudioContent/Waves/Water/Spot_IceRealm_FrozenLake_IceShifts_04}}
 {id=WavePlayer_4, properties {soundWaveAssetPtr=/AmbientAudioContent/Waves/SpottedLoops/Spot_IceCastle_Ice_Stalacitte_Break_05}}
]
properties {
 attenuationSettings=/Game/Athena/Sounds/Ambient/Fort_Amb_Water_River_Att
 bOverrideConcurrency=true
 maxDistance=2925
 soundClassObject=/Game/Sounds/FortSoundClassesAndModes/Fort_Water_Ambient
 volumeMultiplier=1
 attenuationOverrides {falloffDistance=2000, nonSpatializedRadiusStart=700, stereoSpread=1000, attenuationShapeExtents {x=925}}
 concurrencyOverrides {maxCount=1, voiceStealReleaseTime=2}
}
