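# Looping ambient cue for the bottom of a huge river waterfall, limited to one playing instance.
# The top two quality levels share a five-branch mixer of close, mid and distant-roar loops, while the lower two each play one pitch-modulated loop.
# Each branch has its own attenuation with listener focus, layering the sound by distance through differing shapes, falloffs, filtering and occlusion.
# The two farthest-reaching layers are pitched up, and one starts after a short random delay.
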SoundCue /AmbientAudioContent/SoundCues/Water/SC_Spot_River_Waterfall_Huge_Bottom_Loop_Cue
root=QualityLevel_1
nodes [
 {
  id=Attenuation_10
  children [Modulator_4]
  properties {
   bOverrideAttenuation=true
   attenuationOverrides {
    bAttenuateWithLPF=true
    bEnableFocusInterpolation=true
    bEnableListenerFocus=true
    distanceAlgorithm=Logarithmic
    falloffDistance=7500
    focusAttackInterpSpeed=0.2
    lPFFrequencyAtMax=1000
    lPFRadiusMax=4000
    lPFRadiusMin=1000
    nonFocusVolumeAttenuation=0.8
    nonSpatializedRadiusStart=3000
    occlusionInterpolationTime=1
    occlusionLowPassFilterFrequency=1000
    occlusionTraceChannel=ECC_GameTraceChannel11
    stereoSpread=500
    attenuationShapeExtents {x=2000}
   }
  }
 }
 {
  id=Attenuation_11
  children [Mixer_4]
  properties {
   bOverrideAttenuation=true
   attenuationOverrides {
    attenuationShape=Capsule
    bAttenuateWithLPF=true
    bEnableFocusInterpolation=true
    bEnableListenerFocus=true
    bEnableLogFrequencyScaling=true
    bEnableOcclusion=true
    distanceAlgorithm=Logarithmic
    falloffDistance=8000
    focusAttackInterpSpeed=0.01
    focusAzimuth=90
    focusReleaseInterpSpeed=0.01
    focusVolumeAttenuation=1.2
    lPFFrequencyAtMax=1000
    lPFRadiusMax=8000
    lPFRadiusMin=1000
    nonFocusAzimuth=120
    nonFocusVolumeAttenuation=0.75
    nonSpatializedRadiusStart=500
    occlusionInterpolationTime=10
    occlusionLowPassFilterFrequency=1000
    occlusionTraceChannel=ECC_GameTraceChannel6
    occlusionVolumeAttenuation=0.25
    reverbWetLevelMin=0.4
    stereoSpread=1000
    attenuationShapeExtents=<1500 1500>
   }
  }
 }
 {
  id=Attenuation_5
  children [WavePlayer_10]
  properties {
   bOverrideAttenuation=true
   attenuationOverrides {
    attenuationShape=Capsule
    bAttenuateWithLPF=true
    bEnableFocusInterpolation=true
    bEnableListenerFocus=true
    distanceAlgorithm=Logarithmic
    falloffDistance=8000
    focusAttackInterpSpeed=0.1
    lPFFrequencyAtMax=1000
    lPFRadiusMax=4000
    lPFRadiusMin=1000
    nonFocusVolumeAttenuation=0.8
    nonSpatializedRadiusStart=2000
    occlusionInterpolationTime=1
    occlusionLowPassFilterFrequency=1000
    occlusionTraceChannel=ECC_GameTraceChannel11
    stereoSpread=500
    attenuationShapeExtents=<1500 1500>
   }
  }
 }
 {
  id=Attenuation_6
  children [Modulator_5]
  properties {
   bOverrideAttenuation=true
   attenuationOverrides {
    bEnableFocusInterpolation=true
    bEnableListenerFocus=true
    bEnableOcclusion=true
    distanceAlgorithm=Custom
    falloffDistance=20000
    lPFFrequencyAtMax=1000
    lPFRadiusMax=4000
    lPFRadiusMin=1000
    nonFocusVolumeAttenuation=0.8
    nonSpatializedRadiusStart=5000
    occlusionInterpolationTime=3
    occlusionLowPassFilterFrequency=1000
    occlusionTraceChannel=ECC_GameTraceChannel11
    stereoSpread=10000
    attenuationShapeExtents {x=0}
    customAttenuationCurve {editorCurveData {keys [
       <at 0 0>
       {arriveTangent=0.028, interpMode=RCIM_Cubic, leaveTangent=0.028, tangentMode=RCTM_User, time=0.2209, value=1}
       {arriveTangent=-0.2554, interpMode=RCIM_Cubic, leaveTangent=-0.2554, tangentMode=RCTM_User, time=0.4, value=0.97}
       <at 1 0>
      ]}}
   }
  }
 }
 {
  id=Attenuation_7
  children [WavePlayer_11]
  properties {
   bOverrideAttenuation=true
   attenuationOverrides {
    attenuationShape=Capsule
    bAttenuateWithLPF=true
    bEnableFocusInterpolation=true
    bEnableListenerFocus=true
    distanceAlgorithm=Logarithmic
    falloffDistance=3000
    focusAttackInterpSpeed=0.1
    focusVolumeAttenuation=1.2
    lPFFrequencyAtMax=1000
    lPFRadiusMax=4000
    lPFRadiusMin=1000
    nonFocusVolumeAttenuation=0.8
    nonSpatializedRadiusStart=500
    occlusionInterpolationTime=5
    occlusionLowPassFilterFrequency=1000
    occlusionTraceChannel=ECC_GameTraceChannel11
    occlusionVolumeAttenuation=0.5
    reverbWetLevelMin=0.4
    stereoSpread=500
    attenuationShapeExtents {x=1500}
   }
  }
 }
 {
  id=Attenuation_8
  children [Modulator_6]
  properties {
   bOverrideAttenuation=true
   attenuationOverrides {
    bAttenuateWithLPF=true
    bEnableFocusInterpolation=true
    bEnableListenerFocus=true
    bEnableOcclusion=true
    distanceAlgorithm=Custom
    falloffDistance=12000
    focusAttackInterpSpeed=0.2
    focusReleaseInterpSpeed=0.2
    hPFFrequencyAtMax=300
    lPFFrequencyAtMax=1000
    lPFRadiusMax=12000
    lPFRadiusMin=5000
    nonFocusVolumeAttenuation=0.8
    nonSpatializedRadiusStart=3750
    occlusionInterpolationTime=3
    occlusionLowPassFilterFrequency=1000
    occlusionTraceChannel=ECC_GameTraceChannel11
    stereoSpread=10000
    attenuationShapeExtents {x=2000}
    customAttenuationCurve {editorCurveData {keys [
       <at 0 1>
       {arriveTangent=0.028, interpMode=RCIM_Cubic, leaveTangent=0.028, tangentMode=RCTM_User, time=0.2209, value=1}
       {arriveTangent=-0.2554, interpMode=RCIM_Cubic, leaveTangent=-0.2554, tangentMode=RCTM_User, time=0.4, value=0.97}
       <at 1 0>
      ]}}
   }
  }
 }
 {
  id=Attenuation_9
  children [Modulator_2]
  properties {
   bOverrideAttenuation=true
   attenuationOverrides {
    bAttenuateWithLPF=true
    bEnableFocusInterpolation=true
    bEnableListenerFocus=true
    distanceAlgorithm=Logarithmic
    falloffDistance=7500
    focusAttackInterpSpeed=0.2
    lPFFrequencyAtMax=1000
    lPFRadiusMax=4000
    lPFRadiusMin=1000
    nonFocusVolumeAttenuation=0.8
    nonSpatializedRadiusStart=3000
    occlusionInterpolationTime=1
    occlusionLowPassFilterFrequency=1000
    occlusionTraceChannel=ECC_GameTraceChannel11
    stereoSpread=500
    attenuationShapeExtents {x=2000}
   }
  }
 }
 {id=Delay_1, children [Attenuation_8], properties {delayMax=0.5, delayMin=0.2}}
 {id=Mixer_3, children [Attenuation_7, Attenuation_5, Attenuation_11, Delay_1, Attenuation_6], properties {inputVolume [1.5,1.5,6,4,1]}}
 {id=Mixer_4, children [WavePlayer_12, WavePlayer_13], properties {inputVolume [1,1]}}
 {id=Modulator_2, children [WavePlayer_14], properties {pitchMax=1.1, pitchMin=0.9}}
 {id=Modulator_4, children [WavePlayer_15], properties {pitchMax=1.1, pitchMin=0.9}}
 {id=Modulator_5, children [WavePlayer_1], properties {pitchMax=1.3, pitchMin=1.2}}
 {id=Modulator_6, children [WavePlayer_2], properties {pitchMax=1.3, pitchMin=1.2}}
 {id=QualityLevel_1, children [Mixer_3, Mixer_3, Attenuation_9, Attenuation_10]}
 {
  id=WavePlayer_1
  properties {bLooping=true, soundWaveAssetPtr=/AmbientAudioContent/Waves/Water/Spot_RealityFalls_Waterfall_Roar_Medium_Distant_01}
 }
 {
  id=WavePlayer_10
  properties {bLooping=true, soundWaveAssetPtr=/AmbientAudioContent/Waves/Water/Spot_RealityFalls_Waterfall_Large_Bottom_Exterior_Loop}
 }
 {
  id=WavePlayer_11
  properties {bLooping=true, soundWaveAssetPtr=/AmbientAudioContent/Waves/Water/Spot_RealityFalls_Waterfall_Large_Bottom_Exterior_Close_Loop}
 }
 {
  id=WavePlayer_12
  properties {bLooping=true, soundWaveAssetPtr=/AmbientAudioContent/Waves/Water/Spot_RealityFalls_Waterfall_Large_Bottom_Exterior_Close_Loop_02}
 }
 {
  id=WavePlayer_13
  properties {bLooping=true, soundWaveAssetPtr=/AmbientAudioContent/Waves/Water/Spot_RealityFalls_Waterfall_Large_Bottom_Exterior_Close_Loop_03}
 }
 {
  id=WavePlayer_14
  properties {bLooping=true, soundWaveAssetPtr=&Spot_RealityFalls_Waterfall_Large_Bottom_Exterior_Loop}
 }
 {
  id=WavePlayer_15
  properties {bLooping=true, soundWaveAssetPtr=&Spot_RealityFalls_Waterfall_Large_Bottom_Exterior_Loop}
 }
 {id=WavePlayer_2, properties {bLooping=true, soundWaveAssetPtr=/AmbientAudioContent/Waves/Water/Spot_Water_Waterfall_Large_Bot_Loop_01}}
]
properties {
 bOverrideAttenuation=true
 bOverrideConcurrency=true
 maxDistance=20000
 pitchMultiplier=1.1
 soundClassObject=/Game/Sounds/FortSoundClassesAndModes/Fort_Water_Ambient
 attenuationOverrides {
  bAttenuateWithLPF=true
  bEnableOcclusion=true
  bEnableReverbSend=false
  bEnableSubmixSends=true
  distanceAlgorithm=Logarithmic
  falloffDistance=3000
  hPFFrequencyAtMax=200
  lPFFrequencyAtMax=1000
  lPFRadiusMax=3000
  lPFRadiusMin=1000
  nonFocusVolumeAttenuation=0.5
  nonSpatializedRadiusMode=Direct2D
  occlusionTraceChannel=ECC_GameTraceChannel6
  occlusionVolumeAttenuation=0.5
  attenuationShapeExtents {x=500}
  submixSendSettings [{
    bEnableHPFCutoff=false
    bEnableLPFCutoff=false
    disableManualSendClamp=false
    hPFCutoff=20
    lPFCutoff=20000
    maxSendDistance=1200
    maxSendLevel=0.5
    minSendDistance=400
    minSendLevel=0.2
    sendLevel=0.2
    sendLevelControlMethod=Linear
    soundSubmix=&Fortnite_Reverb_Distant
   }]
 }
 concurrencyOverrides {maxCount=1, resolutionRule=StopFarthestThenPreventNew}
}
