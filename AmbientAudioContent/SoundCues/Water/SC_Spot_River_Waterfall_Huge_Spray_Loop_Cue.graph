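# Looping ambient sound cue for the spray of a huge river waterfall, played in the outdoor ambience sound class.
# A quality-level switch feeds the first tier from a random pick of two looping waves through a modulator.
# The other three tiers each play one pitch-modulated loop through low-platform attenuation, the last two using an optimized wave.
# Concurrency allows one instance, stopping the farthest first.
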SoundCue /AmbientAudioContent/SoundCues/Water/SC_Spot_River_Waterfall_Huge_Spray_Loop_Cue
root=QualityLevel_1
nodes [
 {
  id=Attenuation_0
  children [Modulator_4]
  properties {attenuationSettings=/Game/Sounds/Attenuation/Fort_Ambient_Capsule_Waterfall_Huge_Spray_Low_Plat_Att}
 }
 {
  id=Attenuation_1
  children [Modulator_5]
  properties {attenuationSettings=&Fort_Ambient_Capsule_Waterfall_Huge_Spray_Low_Plat_Att}
 }
 {
  id=Attenuation_2
  children [Modulator_6]
  properties {attenuationSettings=&Fort_Ambient_Capsule_Waterfall_Huge_Spray_Low_Plat_Att}
 }
 {id=Modulator_2, children [Random_0]}
 {id=Modulator_4, children [WavePlayer_10], properties {pitchMax=1.1, pitchMin=0.9}}
 {id=Modulator_5, children [WavePlayer_11], properties {pitchMax=1.1, pitchMin=0.9}}
 {id=Modulator_6, children [WavePlayer_12], properties {pitchMax=1.1, pitchMin=0.9}}
 {id=QualityLevel_1, children [Modulator_2, Attenuation_0, Attenuation_1, Attenuation_2]}
 {id=Random_0, children [WavePlayer_2,WavePlayer_1], properties {weights [1,1]}}
 {id=WavePlayer_1, properties {bLooping=true, soundWaveAssetPtr=/AmbientAudioContent/Waves/Water/Spot_waterfall_new_ST_03}}
 {id=WavePlayer_10, properties {bLooping=true, soundWaveAssetPtr=/AmbientAudioContent/Waves/Water/Spot_waterfall_new_ST_01}}
 {id=WavePlayer_11, properties {bLooping=true, soundWaveAssetPtr=/AmbientAudioContent/Waves/Water/Spot_waterfall_new_ST_01_opt}}
 {id=WavePlayer_12, properties {bLooping=true, soundWaveAssetPtr=&Spot_waterfall_new_ST_01_opt}}
 {id=WavePlayer_2, properties {bLooping=true, soundWaveAssetPtr=&Spot_waterfall_new_ST_01}}
]
properties {
 attenuationSettings=/Game/Sounds/Attenuation/Fort_Ambient_Capsule_Waterfall_Huge_Spray_Att
 bOverrideConcurrency=true
 maxDistance=23500
 soundClassObject=&Fort_Ambient_Outdoors_Athena
 volumeMultiplier=0.18
 concurrencyOverrides {maxCount=1, resolutionRule=StopFarthestThenPreventNew}
}
