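# Looping ambient cue for the top of a huge river waterfall, in the water ambient sound class and limited to one concurrent instance.
# A quality-level node plays either a layered mix in its first slot or one pitch-modulated close-roar loop in the others.
# The mix blends near river and waterfall loops with a long-range, custom-curve bed of distant roars and a delayed waterfall loop.
# Each branch overrides attenuation with distance low-pass and listener focus.
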SoundCue /AmbientAudioContent/SoundCues/Water/SC_Spot_River_Waterfall_Huge_Top_Loop_Cue
root=QualityLevel_0
nodes [
 {
  id=Attenuation_0
  children [Modulator_0]
  properties {
   bOverrideAttenuation=true
   attenuationOverrides {
    bAttenuateWithLPF=true
    bEnableFocusInterpolation=true
    bEnableListenerFocus=true
    bEnableOcclusion=true
    distanceAlgorithm=Logarithmic
    falloffDistance=7500
    lPFFrequencyAtMax=1000
    lPFRadiusMax=4000
    lPFRadiusMin=1000
    nonFocusVolumeAttenuation=0.8
    nonSpatializedRadiusStart=3000
    occlusionInterpolationTime=2
    occlusionLowPassFilterFrequency=1000
    occlusionTraceChannel=ECC_GameTraceChannel11
    stereoSpread=500
    attenuationShapeExtents {x=2000}
   }
  }
 }
 {
  id=Attenuation_1
  children [Modulator_1]
  properties {
   bOverrideAttenuation=true
   attenuationOverrides {
    bAttenuateWithLPF=true
    bEnableFocusInterpolation=true
    bEnableListenerFocus=true
    bEnableOcclusion=true
    distanceAlgorithm=Logarithmic
    falloffDistance=7500
    lPFFrequencyAtMax=1000
    lPFRadiusMax=4000
    lPFRadiusMin=1000
    nonFocusVolumeAttenuation=0.8
    nonSpatializedRadiusStart=3000
    occlusionInterpolationTime=2
    occlusionLowPassFilterFrequency=1000
    occlusionTraceChannel=ECC_GameTraceChannel11
    stereoSpread=500
    attenuationShapeExtents {x=2000}
   }
  }
 }
 {
  id=Attenuation_4
  children [Modulator_3]
  properties {
   bOverrideAttenuation=true
   attenuationOverrides {
    bAttenuateWithLPF=true
    bEnableFocusInterpolation=true
    bEnableListenerFocus=true
    bEnableOcclusion=true
    distanceAlgorithm=Logarithmic
    falloffDistance=7500
    lPFFrequencyAtMax=1000
    lPFRadiusMax=4000
    lPFRadiusMin=1000
    nonFocusVolumeAttenuation=0.8
    nonSpatializedRadiusStart=3000
    occlusionInterpolationTime=2
    occlusionLowPassFilterFrequency=1000
    occlusionTraceChannel=ECC_GameTraceChannel11
    stereoSpread=500
    attenuationShapeExtents {x=2000}
   }
  }
 }
 {
  id=Attenuation_5
  children [Mixer_1]
  properties {
   bOverrideAttenuation=true
   attenuationOverrides {
    bAttenuateWithLPF=true
    bEnableFocusInterpolation=true
    bEnableListenerFocus=true
    bEnableOcclusion=true
    distanceAlgorithm=Custom
    falloffDistance=25000
    focusAttackInterpSpeed=0.2
    focusAzimuth=90
    focusReleaseInterpSpeed=0.2
    lPFRadiusMax=25000
    lPFRadiusMin=8000
    nonFocusAzimuth=120
    nonSpatializedRadiusStart=5000
    occlusionInterpolationTime=3
    occlusionLowPassFilterFrequency=1000
    occlusionTraceChannel=ECC_GameTraceChannel6
    occlusionVolumeAttenuation=0.75
    stereoSpread=800
    customAttenuationCurve {editorCurveData {keys [
       <at 0 0.75>
       {arriveTangent=-0.2554, interpMode=RCIM_Cubic, leaveTangent=-0.2554, tangentMode=RCTM_User, time=0.1, value=1}
       {arriveTangent=-1.1111, interpMode=RCIM_Cubic, leaveTangent=-1.1111, time=0.4, value=0.33}
       {arriveTangent=-1.3189, interpMode=RCIM_Cubic, leaveTangent=-1.3189, tangentMode=RCTM_User, time=1, value=0}
      ]}}
   }
  }
 }
 {
  id=Attenuation_6
  children [Mixer_0]
  properties {
   bOverrideAttenuation=true
   attenuationOverrides {
    bAttenuateWithLPF=true
    bEnableFocusInterpolation=true
    bEnableListenerFocus=true
    bEnableLogFrequencyScaling=true
    distanceAlgorithm=Logarithmic
    falloffDistance=8000
    focusAttackInterpSpeed=0.2
    focusReleaseInterpSpeed=0.2
    hPFFrequencyAtMax=300
    lPFFrequencyAtMax=1000
    lPFRadiusMax=8000
    lPFRadiusMin=1500
    nonFocusVolumeAttenuation=0.4
    nonSpatializedRadiusStart=2000
    stereoSpread=1000
    attenuationShapeExtents {x=1500}
   }
  }
 }
 {id=Delay_0, children [WavePlayer_8], properties {delayMax=1, delayMin=0.5}}
 {id=Mixer_0, children [WavePlayer_0,WavePlayer_3], properties {inputVolume [2,3]}}
 {id=Mixer_1, children [WavePlayer_4,WavePlayer_5,Delay_0], properties {inputVolume [1,0.8,5]}}
 {id=Mixer_2, children [Attenuation_6, Attenuation_5], properties {inputVolume [1,1.25]}}
 {id=Modulator_0, children [WavePlayer_6], properties {pitchMax=1.3, pitchMin=1.2}}
 {id=Modulator_1, children [WavePlayer_7], properties {pitchMax=1.1, pitchMin=0.9}}
 {id=Modulator_3, children [WavePlayer_9], properties {pitchMax=1.1, pitchMin=0.9}}
 {id=QualityLevel_0, children [Mixer_2, Attenuation_0, Attenuation_1, Attenuation_4]}
 {
  id=WavePlayer_0
  properties {bLooping=true, soundWaveAssetPtr=/AmbientAudioContent/Waves/Water/Spot_RealityFalls_Waterfall_Large_Top_River_Loop}
 }
 {id=WavePlayer_3, properties {bLooping=true, soundWaveAssetPtr=/AmbientAudioContent/Waves/Water/Spot_Water_Waterfall_Large_Bot_Loop_01}}
 {id=WavePlayer_4, properties {bLooping=true, soundWaveAssetPtr=/AmbientAudioContent/Waves/Water/Spot_RealityFalls_Waterfall_Roar_Distant}}
 {
  id=WavePlayer_5
  properties {bLooping=true, soundWaveAssetPtr=/AmbientAudioContent/Waves/Water/Spot_RealityFalls_Waterfall_Roar_Medium_Distant_01}
 }
 {id=WavePlayer_6, properties {bLooping=true, soundWaveAssetPtr=/AmbientAudioContent/Waves/Water/Spot_RealityFalls_Waterfall_Roar_Close}}
 {id=WavePlayer_7, properties {bLooping=true, soundWaveAssetPtr=&Spot_RealityFalls_Waterfall_Roar_Close}}
 {id=WavePlayer_8, properties {bLooping=true, soundWaveAssetPtr=&Spot_Water_Waterfall_Large_Bot_Loop_01}}
 {id=WavePlayer_9, properties {bLooping=true, soundWaveAssetPtr=&Spot_RealityFalls_Waterfall_Roar_Close}}
]
properties {
 bOverrideAttenuation=true
 bOverrideConcurrency=true
 maxDistance=25400
 pitchMultiplier=1.1
 soundClassObject=/Game/Sounds/FortSoundClassesAndModes/Fort_Water_Ambient
 attenuationOverrides {
  bAttenuateWithLPF=true
  bEnableOcclusion=true
  bEnableReverbSend=false
  bEnableSubmixSends=true
  distanceAlgorithm=Logarithmic
  falloffDistance=3000
  hPFFrequencyAtMax=200
  lPFFrequencyAtMax=1000
  lPFRadiusMax=3000
  lPFRadiusMin=1000
  nonFocusVolumeAttenuation=0.5
  nonSpatializedRadiusMode=Direct2D
  occlusionTraceChannel=ECC_GameTraceChannel6
  occlusionVolumeAttenuation=0.5
  attenuationShapeExtents {x=500}
  pluginSettings {sourceDataOverridePluginSettingsArray [null]}
  submixSendSettings [{
    bEnableHPFCutoff=false
    bEnableLPFCutoff=false
    disableManualSendClamp=false
    hPFCutoff=20
    lPFCutoff=20000
    maxSendDistance=1200
    maxSendLevel=0.5
    minSendDistance=400
    minSendLevel=0.2
    sendLevel=0.2
    sendLevelControlMethod=Linear
    soundSubmix=&Fortnite_Reverb_Distant
   }]
 }
 concurrencyOverrides {maxCount=1, resolutionRule=StopFarthestThenPreventNew}
}
