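# Looping spot ambience cue for indoor beach resort pool water, on the indoor ambient sound class.
# A lake shore lapping loop is mixed with a quieter close pool lapping loop, whose volume dips periodically through its own Enveloper.
# Three chained looping Envelopers then add slow pitch drift, irregular volume flutter and a gentle swell, varying the static loops.
# It uses shared pool attenuation, one-voice concurrency and default and distant reverb sends.
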SoundCue /AmbientAudioContent/SoundCues/Water/Spot_BeachResort_Pool_Water_Int_Cue
root=Enveloper_4
nodes [
 {
  id=Enveloper_0
  children [Mixer_0]
  properties {
   bLoop=true
   bLoopIndefinitely=true
   loopEnd=60
   pitchCurve {editorCurveData {keys [
      {interpMode=RCIM_Cubic, time=0}
      <at 15 0.85 RCIM_Cubic>
      {arriveTangent=0.0091, interpMode=RCIM_Cubic, leaveTangent=0.0091, tangentMode=RCTM_User, time=30}
      {interpMode=RCIM_Cubic, time=45}
      {time=60}
     ]}}
   volumeCurve {editorCurveData {keys [<at 0 1>]}}
  }
 }
 {
  id=Enveloper_1
  children [WavePlayer_0]
  properties {
   bLoop=true
   bLoopIndefinitely=true
   loopEnd=10
   pitchCurve {editorCurveData {keys [{time=0}]}}
   volumeCurve {editorCurveData {keys [
      {arriveTangent=-0.01, interpMode=RCIM_Cubic, time=0, value=1}
      <at 3 0.2 RCIM_Cubic>
      {arriveTangent=0.16, leaveTangent=0.16, time=6, value=1}
      {arriveTangent=-0.0008, interpMode=RCIM_Cubic, leaveTangent=-0.0008, tangentMode=RCTM_User, time=8, value=1}
      {interpMode=RCIM_Cubic, leaveTangent=0.01, time=10, value=1}
     ]}}
  }
 }
 {
  id=Enveloper_3
  children [Enveloper_0]
  properties {
   bLoop=true
   bLoopIndefinitely=true
   loopEnd=60
   pitchCurve {editorCurveData {keys [{time=0}]}}
   volumeCurve {editorCurveData {keys [
      {arriveTangent=-0.01, interpMode=RCIM_Cubic, time=0, value=1}
      {arriveTangent=-0.0035, interpMode=RCIM_Cubic, leaveTangent=-0.0035, tangentMode=RCTM_User, time=5.0488, value=0.8292}
      {arriveTangent=0.0016, interpMode=RCIM_Cubic, leaveTangent=0.0016, tangentMode=RCTM_User, time=7.5626, value=0.9613}
      {arriveTangent=0.0008, interpMode=RCIM_Cubic, leaveTangent=0.0008, tangentMode=RCTM_User, time=11.6137, value=0.8426}
      {arriveTangent=0.0016, interpMode=RCIM_Cubic, leaveTangent=0.0016, tangentMode=RCTM_User, time=13.4835, value=0.8282}
      {arriveTangent=0.0009, interpMode=RCIM_Cubic, leaveTangent=0.0009, time=15.3211, value=0.8889}
      {arriveTangent=-0.0011, interpMode=RCIM_Cubic, leaveTangent=-0.0011, tangentMode=RCTM_User, time=19.0928, value=0.8333}
      {arriveTangent=0.0002, interpMode=RCIM_Cubic, leaveTangent=0.0002, time=24.105, value=0.9798}
      {arriveTangent=0.0016, interpMode=RCIM_Cubic, leaveTangent=0.0016, tangentMode=RCTM_User, time=28.5401, value=0.8354}
      {arriveTangent=-0.0003, interpMode=RCIM_Cubic, leaveTangent=-0.0003, time=31.5586, value=0.9035}
      {arriveTangent=0.0009, interpMode=RCIM_Cubic, leaveTangent=0.0009, tangentMode=RCTM_User, time=35.4287, value=0.8335}
      {arriveTangent=-0.0008, interpMode=RCIM_Cubic, leaveTangent=-0.0008, time=38.0791, value=0.9798}
      <at 42.5142 0.8281>
      {arriveTangent=0.0024, interpMode=RCIM_Cubic, leaveTangent=0.0024, tangentMode=RCTM_User, time=45.0728, value=0.8333}
      {arriveTangent=-0.0006, interpMode=RCIM_Cubic, leaveTangent=-0.0006, time=48.1071, value=0.9613}
      {arriveTangent=-0.0056, interpMode=RCIM_Cubic, leaveTangent=-0.0056, tangentMode=RCTM_User, time=53.6403, value=0.8282}
      <at 55.6724 0.9136>
      {arriveTangent=-0.0003, interpMode=RCIM_Cubic, leaveTangent=-0.0003, tangentMode=RCTM_User, time=58.2596, value=0.828}
      <at 60 1 RCIM_Cubic>
     ]}}
  }
 }
 {
  id=Enveloper_4
  children [Enveloper_3]
  properties {
   bLoop=true
   bLoopIndefinitely=true
   loopEnd=20
   pitchCurve {editorCurveData {keys [{time=0}]}}
   volumeCurve {editorCurveData {keys [
      {arriveTangent=-0.01, interpMode=RCIM_Cubic, time=0, value=1}
      <at 10 1.2 RCIM_Cubic>
      {interpMode=RCIM_Cubic, leaveTangent=0.01, time=20, value=1}
     ]}}
  }
 }
 {id=Mixer_0, children [WavePlayer_3,Enveloper_1], properties {inputVolume [1.4,0.8]}}
 {id=WavePlayer_0, properties {bLooping=true, soundWaveAssetPtr=/AmbientAudioContent/Waves/Water/Spot_Water_Swimming_Pool_Lapping_Close}}
 {id=WavePlayer_3, properties {bLooping=true, soundWaveAssetPtr=/AmbientAudioContent/Waves/Water/Spot_Water_Lake_Shore_Lapping_02}}
]
properties {
 attenuationSettings=/Game/Athena/Sounds/Ambient/Fort_Amb_Water_Pool_Att
 bOverrideConcurrency=true
 maxDistance=2425
 soundClassObject=&Fort_Ambient_Indoors_Athena
 volumeMultiplier=0.501
 attenuationOverrides {falloffDistance=2000, nonSpatializedRadiusStart=700, stereoSpread=1000, attenuationShapeExtents {x=925}}
 concurrencyOverrides {maxCount=1, voiceStealReleaseTime=2}
 soundSubmixSends [
  {minSendLevel=0, soundSubmix=&Fortnite_Reverb_Default}
  {minSendLevel=0, sendLevel=0.6, soundSubmix=&Fortnite_Reverb_Distant}
 ]
}
