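# Looping ambient sound cue for the exterior base of a large river waterfall, in the water ambient sound class and limited to one instance.
# The first two quality slots share a five-branch mix of close and distant loops, while the other two each play one pitch-modulated close roar.
# Each branch has its own attenuation with listener focus, varying in shape, falloff curve and occlusion.
# Two mix layers are pitched up, one starting after a short random delay.
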SoundCue /AmbientAudioContent/SoundCues/Water/Spot_River_Waterfall_Large_Bottom_Exterior_02_Cue
root=QualityLevel_0
nodes [
 {
  id=Attenuation_0
  children [WavePlayer_0]
  properties {
   bOverrideAttenuation=true
   attenuationOverrides {
    attenuationShape=Capsule
    bAttenuateWithLPF=true
    bEnableFocusInterpolation=true
    bEnableListenerFocus=true
    distanceAlgorithm=Logarithmic
    falloffDistance=5000
    focusAttackInterpSpeed=0.2
    lPFFrequencyAtMax=1000
    lPFRadiusMax=4000
    lPFRadiusMin=1000
    nonFocusVolumeAttenuation=0.8
    nonSpatializedRadiusStart=2000
    occlusionInterpolationTime=1
    occlusionLowPassFilterFrequency=1000
    occlusionTraceChannel=ECC_GameTraceChannel11
    stereoSpread=500
    attenuationShapeExtents=<1500 1500>
   }
  }
 }
 {
  id=Attenuation_1
  children [WavePlayer_6]
  properties {
   bOverrideAttenuation=true
   attenuationOverrides {
    attenuationShape=Capsule
    bAttenuateWithLPF=true
    bEnableFocusInterpolation=true
    bEnableListenerFocus=true
    distanceAlgorithm=Logarithmic
    falloffDistance=3000
    focusAttackInterpSpeed=0.1
    focusVolumeAttenuation=1.2
    lPFFrequencyAtMax=1000
    lPFRadiusMax=4000
    lPFRadiusMin=1000
    nonFocusVolumeAttenuation=0.8
    nonSpatializedRadiusStart=500
    occlusionInterpolationTime=5
    occlusionLowPassFilterFrequency=1000
    occlusionTraceChannel=ECC_GameTraceChannel11
    occlusionVolumeAttenuation=0.5
    reverbWetLevelMin=0.4
    stereoSpread=500
    attenuationShapeExtents {x=1500}
   }
  }
 }
 {
  id=Attenuation_2
  children [Modulator_4]
  properties {
   bOverrideAttenuation=true
   attenuationOverrides {
    bEnableFocusInterpolation=true
    bEnableListenerFocus=true
    bEnableOcclusion=true
    distanceAlgorithm=Custom
    falloffDistance=4000
    lPFFrequencyAtMax=1000
    lPFRadiusMax=4000
    lPFRadiusMin=1000
    nonFocusVolumeAttenuation=0.8
    nonSpatializedRadiusStart=4000
    occlusionInterpolationTime=3
    occlusionLowPassFilterFrequency=1000
    occlusionTraceChannel=ECC_GameTraceChannel11
    stereoSpread=10000
    attenuationShapeExtents {x=0}
    customAttenuationCurve {editorCurveData {keys [
       <at 0 0.4399>
       {arriveTangent=0.028, interpMode=RCIM_Cubic, leaveTangent=0.028, tangentMode=RCTM_User, time=0.2209, value=1}
       {arriveTangent=-0.2554, interpMode=RCIM_Cubic, leaveTangent=-0.2554, tangentMode=RCTM_User, time=0.4, value=0.97}
       <at 1 0>
      ]}}
   }
  }
 }
 {
  id=Attenuation_3
  children [Modulator_2]
  properties {
   bOverrideAttenuation=true
   attenuationOverrides {
    bEnableFocusInterpolation=true
    bEnableListenerFocus=true
    bEnableOcclusion=true
    distanceAlgorithm=Custom
    falloffDistance=4000
    lPFFrequencyAtMax=1000
    lPFRadiusMax=4000
    lPFRadiusMin=1000
    nonFocusVolumeAttenuation=0.8
    nonSpatializedRadiusStart=4000
    occlusionInterpolationTime=3
    occlusionLowPassFilterFrequency=1000
    occlusionTraceChannel=ECC_GameTraceChannel11
    stereoSpread=10000
    attenuationShapeExtents {x=0}
    customAttenuationCurve {editorCurveData {keys [
       <at 0 0.75>
       {arriveTangent=0.028, interpMode=RCIM_Cubic, leaveTangent=0.028, tangentMode=RCTM_User, time=0.2209, value=1}
       {arriveTangent=-0.2554, interpMode=RCIM_Cubic, leaveTangent=-0.2554, tangentMode=RCTM_User, time=0.4, value=0.97}
       <at 1 0>
      ]}}
   }
  }
 }
 {
  id=Attenuation_4
  children [Mixer_1]
  properties {
   bOverrideAttenuation=true
   attenuationOverrides {
    attenuationShape=Capsule
    bAttenuateWithLPF=true
    bEnableFocusInterpolation=true
    bEnableListenerFocus=true
    bEnableOcclusion=true
    distanceAlgorithm=Logarithmic
    falloffDistance=3000
    focusAttackInterpSpeed=0.01
    focusReleaseInterpSpeed=0.01
    focusVolumeAttenuation=1.2
    lPFFrequencyAtMax=1000
    lPFRadiusMax=4000
    lPFRadiusMin=1000
    nonFocusVolumeAttenuation=0.75
    nonSpatializedRadiusStart=500
    occlusionInterpolationTime=10
    occlusionLowPassFilterFrequency=1000
    occlusionTraceChannel=ECC_GameTraceChannel11
    occlusionVolumeAttenuation=0.25
    reverbWetLevelMin=0.4
    stereoSpread=1000
    attenuationShapeExtents=<1500 500>
   }
  }
 }
 {
  id=Attenuation_6
  children [Modulator_1]
  properties {
   bOverrideAttenuation=true
   attenuationOverrides {
    bAttenuateWithLPF=true
    bEnableFocusInterpolation=true
    bEnableListenerFocus=true
    distanceAlgorithm=Logarithmic
    falloffDistance=7500
    focusAttackInterpSpeed=0.2
    lPFFrequencyAtMax=1000
    lPFRadiusMax=4000
    lPFRadiusMin=1000
    nonFocusVolumeAttenuation=0.8
    nonSpatializedRadiusStart=3000
    occlusionInterpolationTime=1
    occlusionLowPassFilterFrequency=1000
    occlusionTraceChannel=ECC_GameTraceChannel11
    stereoSpread=500
    attenuationShapeExtents {x=2000}
   }
  }
 }
 {
  id=Attenuation_7
  children [Modulator_3]
  properties {
   bOverrideAttenuation=true
   attenuationOverrides {
    bAttenuateWithLPF=true
    bEnableFocusInterpolation=true
    bEnableListenerFocus=true
    distanceAlgorithm=Logarithmic
    falloffDistance=7500
    focusAttackInterpSpeed=0.2
    lPFFrequencyAtMax=1000
    lPFRadiusMax=4000
    lPFRadiusMin=1000
    nonFocusVolumeAttenuation=0.8
    nonSpatializedRadiusStart=3000
    occlusionInterpolationTime=1
    occlusionLowPassFilterFrequency=1000
    occlusionTraceChannel=ECC_GameTraceChannel11
    stereoSpread=500
    attenuationShapeExtents {x=2000}
   }
  }
 }
 {id=Delay_0, children [Attenuation_3], properties {delayMax=0.5, delayMin=0.2}}
 {id=Mixer_0, children [Attenuation_0, Attenuation_1, Attenuation_4, Attenuation_2, Delay_0], properties {inputVolume [1,1,5,1,3]}}
 {id=Mixer_1, children [WavePlayer_1,WavePlayer_8], properties {inputVolume [1,1]}}
 {id=Modulator_1, children [WavePlayer_5], properties {pitchMax=1.1, pitchMin=0.9}}
 {id=Modulator_2, children [WavePlayer_2], properties {pitchMax=1.3, pitchMin=1.2}}
 {id=Modulator_3, children [WavePlayer_9], properties {pitchMax=1.1, pitchMin=0.9}}
 {id=Modulator_4, children [WavePlayer_3], properties {pitchMax=1.3, pitchMin=1.2}}
 {id=QualityLevel_0, children [Mixer_0, Mixer_0, Attenuation_6, Attenuation_7]}
 {
  id=WavePlayer_0
  properties {bLooping=true, soundWaveAssetPtr=/AmbientAudioContent/Waves/Water/Spot_RealityFalls_Waterfall_Large_Bottom_Exterior_Loop}
 }
 {
  id=WavePlayer_1
  properties {bLooping=true, soundWaveAssetPtr=/AmbientAudioContent/Waves/Water/Spot_RealityFalls_Waterfall_Large_Bottom_Exterior_Close_Loop_02}
 }
 {id=WavePlayer_2, properties {bLooping=true, soundWaveAssetPtr=/AmbientAudioContent/Waves/Water/Spot_Water_Waterfall_Large_Bot_Loop_01}}
 {
  id=WavePlayer_3
  properties {bLooping=true, soundWaveAssetPtr=/AmbientAudioContent/Waves/Water/Spot_RealityFalls_Waterfall_Roar_Medium_Distant_01}
 }
 {id=WavePlayer_5, properties {bLooping=true, soundWaveAssetPtr=/AmbientAudioContent/Waves/Water/Spot_RealityFalls_Waterfall_Roar_Close}}
 {
  id=WavePlayer_6
  properties {bLooping=true, soundWaveAssetPtr=/AmbientAudioContent/Waves/Water/Spot_RealityFalls_Waterfall_Large_Bottom_Exterior_Close_Loop}
 }
 {
  id=WavePlayer_8
  properties {bLooping=true, soundWaveAssetPtr=/AmbientAudioContent/Waves/Water/Spot_RealityFalls_Waterfall_Large_Bottom_Exterior_Close_Loop_03}
 }
 {id=WavePlayer_9, properties {bLooping=true, soundWaveAssetPtr=&Spot_RealityFalls_Waterfall_Roar_Close}}
]
properties {
 bOverrideConcurrency=true
 maxDistance=9500
 pitchMultiplier=1.1
 soundClassObject=/Game/Sounds/FortSoundClassesAndModes/Fort_Water_Ambient
 volumeMultiplier=1
 attenuationOverrides {
  bAttenuateWithLPF=true
  bEnableListenerFocus=true
  falloffDistance=6000
  lPFFrequencyAtMax=500
  nonSpatializedRadiusStart=700
  stereoSpread=1000
  attenuationShapeExtents {x=925}
 }
 concurrencyOverrides {maxCount=1}
}
