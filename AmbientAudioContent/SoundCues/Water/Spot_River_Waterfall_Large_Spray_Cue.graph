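# Ambient sound cue for the spray of a large river waterfall, spatialized with a waterfall capsule attenuation and limited to one concurrent voice.
# A root quality-level node selects one of four modulated branches.
# The first branch picks randomly between two non-looping waterfall waves, while the others each loop a single wave with slight pitch variation.
# Two of the looping branches share an "opt" version of the wave.
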SoundCue /AmbientAudioContent/SoundCues/Water/Spot_River_Waterfall_Large_Spray_Cue
root=QualityLevel_1
nodes [
 {id=Modulator_0, children [WavePlayer_0], properties {pitchMax=1.1, pitchMin=0.9}}
 {id=Modulator_1, children [WavePlayer_12], properties {pitchMax=1.1, pitchMin=0.9}}
 {id=Modulator_2, children [Random_0]}
 {id=Modulator_3, children [WavePlayer_1], properties {pitchMax=1.1, pitchMin=0.9}}
 {id=QualityLevel_1, children [Modulator_2,Modulator_1,Modulator_0,Modulator_3]}
 {id=Random_0, children [WavePlayer_2,WavePlayer_3], properties {weights [1,1]}}
 {id=WavePlayer_0, properties {bLooping=true, soundWaveAssetPtr=/AmbientAudioContent/Waves/Water/Spot_waterfall_new_ST_01_opt}}
 {id=WavePlayer_1, properties {bLooping=true, soundWaveAssetPtr=&Spot_waterfall_new_ST_01_opt}}
 {id=WavePlayer_12, properties {bLooping=true, soundWaveAssetPtr=/AmbientAudioContent/Waves/Water/Spot_waterfall_new_ST_01}}
 {id=WavePlayer_2, properties {soundWaveAssetPtr=&Spot_waterfall_new_ST_01}}
 {id=WavePlayer_3, properties {soundWaveAssetPtr=/AmbientAudioContent/Waves/Water/Spot_waterfall_new_ST_03}}
]
properties {
 attenuationSettings=/Game/Sounds/Attenuation/Fort_Ambient_Capsule_Waterfall_Large_Spray_Att
 bDebug=true
 bOverrideConcurrency=true
 maxDistance=4100
 pitchMultiplier=1.1
 soundClassObject=&Fort_Ambient_Athena
 volumeMultiplier=0.301
 concurrencyOverrides {maxCount=1, voiceStealReleaseTime=5}
}
