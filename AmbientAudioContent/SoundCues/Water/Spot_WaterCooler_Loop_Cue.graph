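# Spot ambient sound cue for a water cooler, layering a steady hum with intermittent bubbling.
# A mixer blends a pitched-down looping hum with a slightly quieter bubble branch that loops through a random delay, modulation and a random pick of four one-shots.
# The cue plays quietly in the indoor ambient sound class, using hum attenuation, a reverb submix send and a concurrency override that prevents new instances at the limit.
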SoundCue /AmbientAudioContent/SoundCues/Water/Spot_WaterCooler_Loop_Cue
root=Mixer_0
nodes [
 {id=Delay_0, children [Modulator_0], properties {delayMax=8, delayMin=1}}
 {id=Looping_0, children [Delay_0]}
 {id=Mixer_0, children [Looping_0,Modulator_6], properties {inputVolume [0.7,1]}}
 {id=Modulator_0, children [Random_0]}
 {id=Modulator_6, children [WavePlayer_5], properties {pitchMax=0.95, pitchMin=0.85}}
 {id=Random_0, children [WavePlayer_4, WavePlayer_10, WavePlayer_11, WavePlayer_12], properties {weights [1,1,1,1]}}
 {id=WavePlayer_10, properties {soundWaveAssetPtr=/AmbientAudioContent/Waves/OneShots/Spot_Prop_Water_Cooler_Bubble_03}}
 {id=WavePlayer_11, properties {soundWaveAssetPtr=/AmbientAudioContent/Waves/OneShots/Spot_Prop_Water_Cooler_Bubble_02}}
 {id=WavePlayer_12, properties {soundWaveAssetPtr=/AmbientAudioContent/Waves/OneShots/Spot_Prop_Water_Cooler_Bubble_01}}
 {id=WavePlayer_4, properties {soundWaveAssetPtr=/AmbientAudioContent/Waves/OneShots/Spot_Prop_Water_Cooler_Bubble_04}}
 {id=WavePlayer_5, properties {bLooping=true, soundWaveAssetPtr=/AmbientAudioContent/Waves/SpottedLoops/Spot_Prop_Water_Cooler_Hum_Loop}}
]
properties {
 attenuationSettings=/Game/Sounds/Asteria/Environment_Actors/WaterCooler/Spot_WaterCooler_Hum_Att
 bOverrideConcurrency=true
 maxDistance=750
 soundClassObject=/Game/Sounds/FortSoundClassesAndModes/Fort_Ambient_OneShots_Indoors
 volumeMultiplier=0.2
 attenuationOverrides {falloffDistance=2000}
 concurrencyOverrides {resolutionRule=PreventNew}
 concurrencySet [&Weapon_Drop_StopOldest]
 soundSubmixSends [{minSendLevel=0.5, soundSubmix=&Fortnite_Reverb_Default}]
}
