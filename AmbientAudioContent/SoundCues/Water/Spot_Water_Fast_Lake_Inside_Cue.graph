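# Sound cue for an inside lake water ambience spot, playing a looping shore-laps wave in the Fort_Water_Ambient sound class.
# A single looping wave player feeds the output directly, with no randomization, mixing or modulation.
# Distance behavior references the shared Apollo 2D lake attenuation asset; the cue also stores non-spatialized attenuation values and single-voice, stop-oldest concurrency values in its override fields.
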SoundCue /AmbientAudioContent/SoundCues/Water/Spot_Water_Fast_Lake_Inside_Cue
root=WavePlayer_2
nodes [{id=WavePlayer_2, properties {bLooping=true, soundWaveAssetPtr=/AmbientAudioContent/Waves/Water/Spot_Water_Lake_Shore_Laps_Loop_01}}]
properties {
 attenuationSettings=/Game/Athena/Sounds/Water/Attenuations/Apollo_Water_Lake_2D_Att
 maxDistance=4352
 soundClassObject=/Game/Sounds/FortSoundClassesAndModes/Fort_Water_Ambient
 volumeMultiplier=1
 attenuationOverrides {bSpatialize=false, falloffDistance=1500, nonSpatializedRadiusStart=700, stereoSpread=1000, attenuationShapeExtents {x=925}}
 concurrencyOverrides {maxCount=1, resolutionRule=StopOldest}
}
