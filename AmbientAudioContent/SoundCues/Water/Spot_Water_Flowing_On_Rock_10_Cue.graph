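# Looping ambient Sound Cue for a spot emitter of water flowing on rock, in the Fort_Ambient_Water sound class at reduced volume.
# A quality-level root routes all four of its inputs to the same looping wave player, so every quality setting plays identical audio.
# Overridden attenuation uses natural-sound falloff, wide stereo spread, a non-spatialized inner radius and listener focus that favors water the listener faces.
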SoundCue /AmbientAudioContent/SoundCues/Water/Spot_Water_Flowing_On_Rock_10_Cue
root=QualityLevel_0
nodes [
 {id=QualityLevel_0, children [WavePlayer_0,WavePlayer_0,WavePlayer_0,WavePlayer_0]}
 {id=WavePlayer_0, properties {bLooping=true, soundWaveAssetPtr=/AmbientAudioContent/Waves/Water/Spot_Water_Flowing_On_Rock_10}}
]
properties {
 bOverrideAttenuation=true
 maxDistance=2250
 soundClassObject=/Game/Sounds/FortSoundClassesAndModes/Fort_Ambient_Water
 volumeMultiplier=0.5
 attenuationOverrides {
  bEnableListenerFocus=true
  distanceAlgorithm=NaturalSound
  falloffDistance=1500
  focusAzimuth=25
  focusVolumeAttenuation=1.2
  nonFocusAzimuth=45
  nonFocusVolumeAttenuation=0.5
  nonSpatializedRadiusStart=300
  stereoSpread=1000
  attenuationShapeExtents {x=750}
 }
 concurrencyOverrides {maxCount=1, resolutionRule=PreventNew}
}
