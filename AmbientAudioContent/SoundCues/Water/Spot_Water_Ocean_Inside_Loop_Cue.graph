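# Looping ocean water ambience cue, named as the inside variant, that layers an offshore bed with a medium-distance offshore bed.
# A mixer blends both at equal volume, each through its own 2D attenuation, regular for the offshore layer and distant for the other.
# The offshore layer uses the DuckGlobal water sound class, while the distant layer uses the plain water ambient class.
# Quality-level nodes switch each layer between quad and "_st" waves.
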SoundCue /AmbientAudioContent/SoundCues/Water/Spot_Water_Ocean_Inside_Loop_Cue
root=Mixer_0
nodes [
 {id=Attenuation_0, children [QualityLevel_1], properties {attenuationSettings=/AmbientAudioBase/Attenuations/Apollo_Water_Ocean_2D_Att}}
 {
  id=Attenuation_1
  children [QualityLevel_0]
  properties {attenuationSettings=/AmbientAudioBase/Attenuations/Apollo_Water_Ocean_Distant_2D_Att}
 }
 {id=Mixer_0, children [SoundClass_0,SoundClass_1], properties {inputVolume [1,1]}}
 {id=QualityLevel_0, children [WavePlayer_0,WavePlayer_0,WavePlayer_1,WavePlayer_1]}
 {id=QualityLevel_1, children [WavePlayer_18, WavePlayer_18, WavePlayer_2, WavePlayer_2]}
 {
  id=SoundClass_0
  children [Attenuation_0]
  properties {soundClassOverride=/Game/Sounds/FortSoundClassesAndModes/Fort_Water_Ambient_DuckGlobal_PSM}
 }
 {id=SoundClass_1, children [Attenuation_1], properties {soundClassOverride=/Game/Sounds/FortSoundClassesAndModes/Fort_Water_Ambient}}
 {
  id=WavePlayer_0
  properties {bLooping=true, soundWaveAssetPtr=/AmbientAudioContent/Waves/Water/Spot_Water_Ocean_Offshore_Quad_Med_Dist_Loop_01}
 }
 {
  id=WavePlayer_1
  properties {bLooping=true, soundWaveAssetPtr=/AmbientAudioContent/Waves/Water/Spot_Water_Ocean_Offshore_Quad_Med_Dist_Loop_01_st}
 }
 {id=WavePlayer_18, properties {bLooping=true, soundWaveAssetPtr=/AmbientAudioContent/Waves/Water/Spot_Water_Ocean_Offshore_Quad_Loop_01}}
 {id=WavePlayer_2, properties {bLooping=true, soundWaveAssetPtr=/AmbientAudioContent/Waves/Water/Spot_Water_Ocean_Offshore_Quad_Loop_01_st}}
]
properties {
 maxDistance=8192
 soundClassObject=&Fort_Water_Ambient_DuckGlobal_PSM
 volumeMultiplier=1
 attenuationOverrides {bSpatialize=false, falloffDistance=1500, nonSpatializedRadiusStart=700, stereoSpread=1000, attenuationShapeExtents {x=925}}
 concurrencyOverrides {maxCount=1, resolutionRule=StopOldest}
}
