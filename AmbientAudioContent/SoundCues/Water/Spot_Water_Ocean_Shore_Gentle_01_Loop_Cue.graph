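# Looping sound cue for a gentle ocean shore ambience, layered by distance.
# A root mixer blends very close, close and medium layers, each with its own attenuation asset.
# The very close layer mixes laps with small waves under a repeating volume-dip envelope, while the close layer picks quad or stereo lapping by quality level.
# Medium plays distant waves, and a concurrency override limits the cue to one instance.
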SoundCue /AmbientAudioContent/SoundCues/Water/Spot_Water_Ocean_Shore_Gentle_01_Loop_Cue
root=Mixer_2
nodes [
 {
  id=Attenuation_1
  children [QualityLevel_0]
  properties {attenuationSettings=/AmbientAudioBase/Attenuations/Asteria_Water_Ocean_Shore_Close_Att}
 }
 {
  id=Attenuation_7
  children [WavePlayer_6]
  properties {attenuationSettings=/AmbientAudioBase/Attenuations/Asteria_Water_Ocean_Shore_Medium_Att}
 }
 {
  id=Attenuation_9
  children [Mixer_0]
  properties {attenuationSettings=/AmbientAudioBase/Attenuations/Asteria_Water_Ocean_Shore_Very_Close_Att}
 }
 {
  id=Enveloper_0
  children [WavePlayer_7]
  properties {
   bLoop=true
   bLoopIndefinitely=true
   loopEnd=6
   pitchCurve {editorCurveData {keys [{time=0}]}}
   volumeCurve {editorCurveData {keys [
      <at 0 1 RCIM_Cubic>
      {arriveTangent=-0.2917, interpMode=RCIM_Cubic, leaveTangent=-0.2917, time=2, value=0.2719}
      <at 2.7977 0.1839>
      <at 3.5 1>
     ]}}
  }
 }
 {id=Mixer_0, children [WavePlayer_5,Enveloper_0], properties {inputVolume [1,0.75]}}
 {id=Mixer_2, children [Attenuation_9, Attenuation_1, Attenuation_7], properties {inputVolume [0.7,1,0.4]}}
 {id=QualityLevel_0, children [WavePlayer_0,WavePlayer_0,WavePlayer_1,WavePlayer_1]}
 {id=WavePlayer_0, properties {bLooping=true, soundWaveAssetPtr=/AmbientAudioContent/Waves/Water/Spot_Water_Ocean_Gentle_Waves_Lapping_Quad}}
 {id=WavePlayer_1, properties {bLooping=true, soundWaveAssetPtr=/AmbientAudioContent/Waves/Water/Spot_Water_Ocean_Gentle_Waves_Lapping_Stereo}}
 {id=WavePlayer_5, properties {bLooping=true, soundWaveAssetPtr=/AmbientAudioContent/Waves/Water/Spot_Water_Ocean_Shore_Gentle_Laps_01}}
 {id=WavePlayer_6, properties {bLooping=true, soundWaveAssetPtr=/AmbientAudioContent/Waves/Water/Spot_Water_Ocean_Waves_Med_Distant_Loop_01}}
 {id=WavePlayer_7, properties {bLooping=true, soundWaveAssetPtr=/AmbientAudioContent/Waves/Water/Spot_Water_Ocean_Shore_SmWaves_Loop_01}}
]
properties {
 bOverrideConcurrency=true
 maxDistance=8600
 soundClassObject=&Fort_Ambient_Outdoors_Athena
 attenuationOverrides {
  bAttenuateWithLPF=true
  falloffDistance=6000
  lPFFrequencyAtMax=500
  nonSpatializedRadiusStart=700
  stereoSpread=1000
  attenuationShapeExtents {x=925}
 }
 concurrencyOverrides {maxCount=1, voiceStealReleaseTime=5}
}
