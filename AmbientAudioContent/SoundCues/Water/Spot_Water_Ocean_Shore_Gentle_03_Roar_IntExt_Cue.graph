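# Looping ambient cue for a gentle ocean shore, layering laps, waves and distant roar by distance.
# After a random delay and short fade-in, five layers mix, split between outdoor and general ambient sound classes, with very close to distant attenuations.
# Quality level switches close lapping between quad and stereo waves.
# Fixed down-pitching varies the loops, slow looping envelopes modulate the distant layer, and concurrency allows one instance.
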SoundCue /AmbientAudioContent/SoundCues/Water/Spot_Water_Ocean_Shore_Gentle_03_Roar_IntExt_Cue
root=Delay_0
nodes [
 {
  id=Attenuation_0
  children [WavePlayer_5]
  properties {attenuationSettings=/AmbientAudioBase/Attenuations/Asteria_Water_Ocean_Shore_Medium_Att}
 }
 {
  id=Attenuation_1
  children [Modulator_0]
  properties {attenuationSettings=/AmbientAudioBase/Attenuations/Asteria_Water_Ocean_Shore_Close_Att}
 }
 {
  id=Attenuation_2
  children [WavePlayer_3]
  properties {attenuationSettings=/AmbientAudioBase/Attenuations/Asteria_Water_Ocean_Shore_Distant_Att}
 }
 {
  id=Attenuation_6
  children [WavePlayer_7]
  properties {attenuationSettings=&Asteria_Water_Ocean_Shore_Close_Att}
 }
 {
  id=Attenuation_7
  children [Modulator_1]
  properties {attenuationSettings=&Asteria_Water_Ocean_Shore_Medium_Att}
 }
 {
  id=Attenuation_8
  children [Enveloper_3]
  properties {attenuationSettings=&Asteria_Water_Ocean_Shore_Distant_Att}
 }
 {
  id=Attenuation_9
  children [Modulator_2]
  properties {attenuationSettings=/AmbientAudioBase/Attenuations/Asteria_Water_Ocean_Shore_Very_Close_Att}
 }
 {id=Delay_0, children [Enveloper_1], properties {delayMax=1, delayMin=0.2}}
 {
  id=Enveloper_0
  children [Modulator_3]
  properties {
   bLoop=true
   bLoopIndefinitely=true
   loopEnd=30
   pitchCurve {editorCurveData {keys [
      {arriveTangent=-0.0013, interpMode=RCIM_Cubic, leaveTangent=-0.0013, time=0}
      <at 15 0.98 RCIM_Cubic>
      {arriveTangent=0.0013, interpMode=RCIM_Cubic, leaveTangent=0.0013, time=30}
     ]}}
   volumeCurve {editorCurveData {keys [<at 0 1>]}}
  }
 }
 {
  id=Enveloper_1
  children [Mixer_2]
  properties {pitchCurve {editorCurveData {keys [{time=0}]}}, volumeCurve {editorCurveData {keys [<at 0 0>,<at 0.5 1>]}}}
 }
 {
  id=Enveloper_3
  children [Enveloper_0]
  properties {
   bLoop=true
   bLoopIndefinitely=true
   loopEnd=20
   pitchCurve {editorCurveData {keys [{time=0}]}}
   volumeCurve {editorCurveData {keys [
      {arriveTangent=-0.01, interpMode=RCIM_Cubic, leaveTangent=-0.01, time=0, value=1}
      <at 10 0.9 RCIM_Cubic>
      {arriveTangent=0.01, interpMode=RCIM_Cubic, leaveTangent=0.01, time=20, value=1}
     ]}}
  }
 }
 {id=Mixer_0, children [WavePlayer_2,WavePlayer_4], properties {inputVolume [1,1]}}
 {id=Mixer_1, children [Attenuation_6, Attenuation_0, Attenuation_2], properties {inputVolume [0.75,0.75,0.75]}}
 {id=Mixer_2, children [SoundClass_0,SoundClass_1,SoundClass_2,SoundClass_3,SoundClass_4], properties {inputVolume [0.8,0.9,0.6,0.4,0.5]}}
 {id=Modulator_0, children [QualityLevel_0], properties {pitchMax=0.9, pitchMin=0.9}}
 {id=Modulator_1, children [WavePlayer_29], properties {pitchMax=0.9, pitchMin=0.9}}
 {id=Modulator_2, children [Mixer_0], properties {pitchMax=0.9, pitchMin=0.9}}
 {id=Modulator_3, children [WavePlayer_6], properties {pitchMax=0.9, pitchMin=0.9}}
 {id=QualityLevel_0, children [WavePlayer_0,WavePlayer_0,WavePlayer_1,WavePlayer_1]}
 {
  id=SoundClass_0
  children [Attenuation_9]
  properties {soundClassOverride=&Fort_Ambient_Outdoors_Athena}
 }
 {id=SoundClass_1, children [Attenuation_1], properties {soundClassOverride=&Fort_Ambient_Athena}}
 {
  id=SoundClass_2
  children [Attenuation_7]
  properties {soundClassOverride=&Fort_Ambient_Outdoors_Athena}
 }
 {
  id=SoundClass_3
  children [Attenuation_8]
  properties {soundClassOverride=&Fort_Ambient_Outdoors_Athena}
 }
 {id=SoundClass_4, children [Mixer_1], properties {soundClassOverride=&Fort_Ambient_Athena}}
 {id=WavePlayer_0, properties {bLooping=true, soundWaveAssetPtr=/AmbientAudioContent/Waves/Water/Spot_Water_Ocean_Gentle_Waves_Lapping_Quad}}
 {id=WavePlayer_1, properties {bLooping=true, soundWaveAssetPtr=/AmbientAudioContent/Waves/Water/Spot_Water_Ocean_Gentle_Waves_Lapping_Stereo}}
 {id=WavePlayer_2, properties {bLooping=true, soundWaveAssetPtr=/AmbientAudioContent/Waves/Water/Spot_Water_Ocean_Shore_Gentle_Laps_02}}
 {
  id=WavePlayer_29
  properties {bLooping=true, soundWaveAssetPtr=/AmbientAudioContent/Waves/Water/Spot_Water_Ocean_Offshore_Quad_Med_Dist_Loop_01}
 }
 {id=WavePlayer_3, properties {bLooping=true, soundWaveAssetPtr=/AmbientAudioContent/Waves/Water/Spot_Water_Ocean_Waves_Distant_Loop_01}}
 {id=WavePlayer_4, properties {bLooping=true, soundWaveAssetPtr=/AmbientAudioContent/Waves/Water/Spot_Water_Ocean_Shore_SmWaves_Loop_01}}
 {id=WavePlayer_5, properties {bLooping=true, soundWaveAssetPtr=/AmbientAudioContent/Waves/Water/Spot_Water_Ocean_Waves_Med_Distant_Loop_01}}
 {id=WavePlayer_6, properties {bLooping=true, soundWaveAssetPtr=&Spot_Water_Ocean_Waves_Med_Distant_Loop_01}}
 {id=WavePlayer_7, properties {bLooping=true, soundWaveAssetPtr=/AmbientAudioContent/Waves/Water/Spot_Water_Ocean_Waves_Close_Loop_01}}
]
properties {
 bOverrideConcurrency=true
 maxDistance=12000
 volumeMultiplier=0.551
 attenuationOverrides {
  bAttenuateWithLPF=true
  falloffDistance=6000
  lPFFrequencyAtMax=500
  nonSpatializedRadiusStart=700
  stereoSpread=1000
  attenuationShapeExtents {x=925}
 }
 concurrencyOverrides {maxCount=1, voiceStealReleaseTime=5}
}
