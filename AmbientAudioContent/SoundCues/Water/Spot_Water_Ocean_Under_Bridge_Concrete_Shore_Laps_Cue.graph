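# Ambient spot sound cue for ocean water lapping a concrete shore under a bridge, in the Fort_Water sound class.
# A looping stereo wave feeds two nested, endlessly looping Envelopers: a 30-second pitch curve and a 20-second gentle volume swell, whose differing periods vary the loop.
# Overrides set capsule attenuation with HRTF spatialization, low-pass occlusion and PreventNew concurrency, plus a send to the exterior ambient source bus.
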SoundCue /AmbientAudioContent/SoundCues/Water/Spot_Water_Ocean_Under_Bridge_Concrete_Shore_Laps_Cue
root=Enveloper_2
nodes [
 {
  id=Enveloper_1
  children [WavePlayer_0]
  properties {
   bLoop=true
   bLoopIndefinitely=true
   loopEnd=30
   pitchCurve {editorCurveData {keys [
      {arriveTangent=-0.0013, interpMode=RCIM_Cubic, leaveTangent=-0.0013, time=0}
      <at 15 0.98 RCIM_Cubic>
      {arriveTangent=0.0013, interpMode=RCIM_Cubic, leaveTangent=0.0013, time=30}
     ]}}
   volumeCurve {editorCurveData {keys [<at 0 1>]}}
  }
 }
 {
  id=Enveloper_2
  children [Enveloper_1]
  properties {
   bLoop=true
   bLoopIndefinitely=true
   loopEnd=20
   pitchCurve {editorCurveData {keys [{time=0}]}}
   volumeCurve {editorCurveData {keys [
      {arriveTangent=-0.01, interpMode=RCIM_Cubic, time=0, value=1}
      <at 10 1.1 RCIM_Cubic>
      {interpMode=RCIM_Cubic, leaveTangent=0.01, time=20, value=1}
     ]}}
  }
 }
 {
  id=WavePlayer_0
  properties {bLooping=true, soundWaveAssetPtr=/AmbientAudioContent/Waves/Water/Spot_Water_Ocean_Under_Bridge_Concrete_Shore_Laps_Stereo}
 }
]
properties {
 bOverrideAttenuation=true
 bOverrideConcurrency=true
 maxDistance=2250
 soundClassObject=/Game/Sounds/FortSoundClassesAndModes/Fort_Water
 volumeMultiplier=0.7
 attenuationOverrides {
  attenuationShape=Capsule
  bEnableOcclusion=true
  bEnableReverbSend=false
  falloffDistance=1250
  nonSpatializedRadiusEnd=600
  occlusionInterpolationTime=2
  occlusionLowPassFilterFrequency=750
  occlusionTraceChannel=ECC_GameTraceChannel11
  occlusionVolumeAttenuation=0.25
  spatializationAlgorithm=SPATIALIZATION_HRTF
  stereoSpread=1024
  attenuationShapeExtents=<1000 600>
 }
 busSends [{audioBus=null, minSendLevel=0, sendLevel=0.6, soundSourceBus=&AmbientAudio_Exterior_SourceBus}]
 concurrencyOverrides {resolutionRule=PreventNew}
}
