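# Spot ambience cue for a river shore with gold, layering water loops with gold-fleck shimmer.
# Three separately attenuated water layers (close 2D, close, medium) each crossfade between two loop sources on the Velocity parameter.
# A focus-attenuated gold layer mixes delayed, pitch-varied random one-shots with a loop that nested Envelopers slowly modulate in pitch and volume.
# An Intensity parameter modulates the combined volume, and concurrency limits instances.
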SoundCue /AmbientAudioContent/SoundCues/Water/Spot_Water_River_Shore_A_Gold_Cue
root=ModulatorContinuous_0
nodes [
 {
  id=Attenuation_0
  children [ParamCrossFade_0]
  properties {attenuationSettings=/AmbientAudioBase/Attenuations/Apollo_Water_River_Close_2D_Att}
 }
 {id=Attenuation_1, children [ParamCrossFade_1], properties {attenuationSettings=/AmbientAudioBase/Attenuations/Apollo_Water_River_Close_Att}}
 {id=Attenuation_2, children [ParamCrossFade_2], properties {attenuationSettings=/AmbientAudioBase/Attenuations/Apollo_Water_River_Medium_Att}}
 {
  id=Attenuation_3
  children [Mixer_1]
  properties {attenuationSettings=/AmbientAudioBase/Attenuations/Apollo_Water_River_Close_Gold_Focus_Att}
 }
 {id=Delay_0, children [Modulator_0], properties {delayMax=6, delayMin=2}}
 {
  id=Enveloper_0
  children [WavePlayer_2]
  properties {
   bLoop=true
   bLoopIndefinitely=true
   loopEnd=30
   pitchCurve {editorCurveData {keys [{arriveTangent=-0.0013, interpMode=RCIM_Cubic, time=0}, <at 15 1.02 RCIM_Cubic>, {interpMode=RCIM_Cubic, leaveTangent=0.0013, time=30}]}}
   volumeCurve {editorCurveData {keys [<at 0 1>]}}
  }
 }
 {
  id=Enveloper_3
  children [Enveloper_0]
  properties {
   bLoop=true
   bLoopIndefinitely=true
   loopEnd=10
   pitchCurve {editorCurveData {keys [{time=0}]}}
   volumeCurve {editorCurveData {keys [
      {arriveTangent=-0.01, interpMode=RCIM_Cubic, time=0, value=1}
      <at 5 0.8 RCIM_Cubic>
      {interpMode=RCIM_Cubic, leaveTangent=0.01, time=10, value=1}
     ]}}
  }
 }
 {id=Looping_0, children [Delay_0]}
 {id=Mixer_0, children [Attenuation_3, Attenuation_0, Attenuation_1, Attenuation_2], properties {inputVolume [1,1,1,1]}}
 {id=Mixer_1, children [Looping_0,Enveloper_3], properties {inputVolume [0.7,0.6]}}
 {id=ModulatorContinuous_0, children [Mixer_0], properties {volumeModulationParams {parameterName=Intensity}}}
 {id=Modulator_0, children [Random_1], properties {pitchMax=1.1, pitchMin=1}}
 {
  id=ParamCrossFade_0
  children [WavePlayer_0,Random_0]
  properties {
   paramName=Velocity
   crossFadeInput [
    {fadeInDistanceEnd=0, fadeInDistanceStart=0, fadeOutDistanceEnd=1, fadeOutDistanceStart=0.5, volume=1}
    {fadeInDistanceEnd=1, fadeInDistanceStart=0.5, fadeOutDistanceEnd=100, fadeOutDistanceStart=100, volume=10}
   ]
  }
 }
 {
  id=ParamCrossFade_1
  children [WavePlayer_6,WavePlayer_7]
  properties {
   paramName=Velocity
   crossFadeInput [
    {fadeInDistanceEnd=0, fadeInDistanceStart=0, fadeOutDistanceEnd=1, fadeOutDistanceStart=0.5, volume=1}
    {fadeInDistanceEnd=1, fadeInDistanceStart=0.5, fadeOutDistanceEnd=100, fadeOutDistanceStart=100, volume=10}
   ]
  }
 }
 {
  id=ParamCrossFade_2
  children [WavePlayer_8,WavePlayer_9]
  properties {
   paramName=Velocity
   crossFadeInput [
    {fadeInDistanceEnd=0, fadeInDistanceStart=0, fadeOutDistanceEnd=100, fadeOutDistanceStart=1, volume=1}
    {fadeInDistanceEnd=1, fadeInDistanceStart=0.5, fadeOutDistanceEnd=100, fadeOutDistanceStart=100, volume=5}
   ]
  }
 }
 {id=Random_0, children [WavePlayer_1,WavePlayer_5], properties {weights [1,1]}}
 {
  id=Random_1
  children [
   WavePlayer_3
   WavePlayer_4
   WavePlayer_10
   WavePlayer_11
   WavePlayer_12
   WavePlayer_13
   WavePlayer_14
   WavePlayer_15
   WavePlayer_16
   WavePlayer_17
   WavePlayer_18
   WavePlayer_19
   WavePlayer_20
   WavePlayer_21
   WavePlayer_22
  ]
  properties {weights [1,1,1,1,1,1,1,1,1,1,1,1,1,1,1]}
 }
 {id=WavePlayer_0, properties {bLooping=true, soundWaveAssetPtr=/AmbientAudioContent/Waves/Water/Spot_Water_Lake_Shore_Laps_Loop_01}}
 {id=WavePlayer_1, properties {bLooping=true, soundWaveAssetPtr=/AmbientAudioContent/Waves/Water/Spot_Water_River_Large_Shore_Rapids_Loop_01}}
 {id=WavePlayer_10, properties {soundWaveAssetPtr=/AmbientAudioContent/Waves/OneShots/Spot_River_Gold_Fleck_Swt_03}}
 {id=WavePlayer_11, properties {soundWaveAssetPtr=/AmbientAudioContent/Waves/OneShots/Spot_River_Gold_Fleck_Swt_04}}
 {id=WavePlayer_12, properties {soundWaveAssetPtr=/AmbientAudioContent/Waves/OneShots/Spot_River_Gold_Fleck_Swt_05}}
 {id=WavePlayer_13, properties {soundWaveAssetPtr=/AmbientAudioContent/Waves/OneShots/Spot_River_Gold_Fleck_Swt_06}}
 {id=WavePlayer_14, properties {soundWaveAssetPtr=/AmbientAudioContent/Waves/OneShots/Spot_River_Gold_Fleck_Swt_07}}
 {id=WavePlayer_15, properties {soundWaveAssetPtr=/AmbientAudioContent/Waves/OneShots/Spot_River_Gold_Fleck_Swt_08}}
 {id=WavePlayer_16, properties {soundWaveAssetPtr=/AmbientAudioContent/Waves/OneShots/Spot_River_Gold_Fleck_Swt_09}}
 {id=WavePlayer_17, properties {soundWaveAssetPtr=/AmbientAudioContent/Waves/OneShots/Spot_River_Gold_Fleck_Swt_10}}
 {id=WavePlayer_18, properties {soundWaveAssetPtr=/AmbientAudioContent/Waves/OneShots/Spot_River_Gold_Fleck_Swt_11}}
 {id=WavePlayer_19, properties {soundWaveAssetPtr=/AmbientAudioContent/Waves/OneShots/Spot_River_Gold_Fleck_Swt_12}}
 {id=WavePlayer_2, properties {bLooping=true, soundWaveAssetPtr=/AmbientAudioContent/Waves/SpottedLoops/Spot_River_Gold_Fleck_Loop}}
 {id=WavePlayer_20, properties {soundWaveAssetPtr=/AmbientAudioContent/Waves/OneShots/Spot_River_Gold_Fleck_Swt_13}}
 {id=WavePlayer_21, properties {soundWaveAssetPtr=/AmbientAudioContent/Waves/OneShots/Spot_River_Gold_Fleck_Swt_14}}
 {id=WavePlayer_22, properties {soundWaveAssetPtr=/AmbientAudioContent/Waves/OneShots/Spot_River_Gold_Fleck_Swt_15}}
 {id=WavePlayer_3, properties {soundWaveAssetPtr=/AmbientAudioContent/Waves/OneShots/Spot_River_Gold_Fleck_Swt_01}}
 {id=WavePlayer_4, properties {soundWaveAssetPtr=/AmbientAudioContent/Waves/OneShots/Spot_River_Gold_Fleck_Swt_02}}
 {id=WavePlayer_5, properties {bLooping=true, soundWaveAssetPtr=/AmbientAudioContent/Waves/Water/Spot_Water_River_Large_Shore_Rapids_Loop_02}}
 {id=WavePlayer_6, properties {bLooping=true, soundWaveAssetPtr=/AmbientAudioContent/Waves/Water/Spot_Water_Stream_Med_Close_Loop_02}}
 {id=WavePlayer_7, properties {bLooping=true, soundWaveAssetPtr=/AmbientAudioContent/Waves/Water/Spot_Water_River_Large_Close_Loop_01}}
 {id=WavePlayer_8, properties {bLooping=true, soundWaveAssetPtr=/AmbientAudioContent/Waves/Water/Spot_Water_River_Close_Loop_01}}
 {id=WavePlayer_9, properties {bLooping=true, soundWaveAssetPtr=/AmbientAudioContent/Waves/Water/Spot_Water_River_Large_Distant_Loop_01}}
]
properties {
 maxDistance=6000
 soundClassObject=/Game/Sounds/FortSoundClassesAndModes/Fort_Water_Ambient
 volumeMultiplier=0.85
 attenuationOverrides {
  bAttenuateWithLPF=true
  bEnableListenerFocus=true
  falloffDistance=6000
  lPFFrequencyAtMax=500
  nonSpatializedRadiusStart=700
  stereoSpread=1000
  attenuationShapeExtents {x=925}
 }
 concurrencyOverrides {maxCount=1, resolutionRule=PreventNew}
 concurrencySet [/Game/Sounds/Concurrency_Rules/Waterfall_StopFarthestOldest_Conc]
}
