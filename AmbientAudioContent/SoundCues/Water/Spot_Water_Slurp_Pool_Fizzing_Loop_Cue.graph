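# Looping ambient sound cue for a fizzing, slurping water pool, routed to the Fort_Water_Ambient sound class.
# A single looping Wave Player feeds the output directly, with no randomization or mixing.
# Distance falloff comes from a shared close-detail lake attenuation asset, as the cue's own attenuation values are not switched on as an override.
# A local concurrency override caps simultaneous instances at a small count, taking precedence over the assigned waterfall concurrency asset.
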SoundCue /AmbientAudioContent/SoundCues/Water/Spot_Water_Slurp_Pool_Fizzing_Loop_Cue
root=WavePlayer_2
nodes [{id=WavePlayer_2, properties {bLooping=true, soundWaveAssetPtr=/AmbientAudioContent/Waves/Water/Spot_Water_Slurp_Pool_Flizzing_Inside_Loop}}]
properties {
 attenuationSettings=/AmbientAudioBase/Attenuations/ATT_Water_Lake_Close_Detail_NoFocus_Inner
 bOverrideConcurrency=true
 maxDistance=1375
 soundClassObject=/Game/Sounds/FortSoundClassesAndModes/Fort_Water_Ambient
 volumeMultiplier=1
 attenuationOverrides {
  bAttenuateWithLPF=true
  falloffDistance=750
  lPFFrequencyAtMax=500
  nonSpatializedRadiusEnd=500
  nonSpatializedRadiusStart=750
  stereoSpread=1000
  attenuationShapeExtents {x=300}
 }
 concurrencyOverrides {maxCount=2}
 concurrencySet [/Game/Sounds/Concurrency_Rules/Waterfall_StopFarthestOldest_Conc]
}
