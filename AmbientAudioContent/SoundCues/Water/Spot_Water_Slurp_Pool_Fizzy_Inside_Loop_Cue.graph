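# Sound cue for a looping indoor spot ambience of a fizzy, slurping water pool, routed to the Fort water ambient sound class.
# Its node graph is empty with no root, so the file shows only playback settings and no wave or loop.
# Custom attenuation adds distance low-pass filtering, near-source non-spatialized blending and occlusion filter settings.
# Overridden concurrency caps voices at two, superseding the referenced waterfall concurrency set.
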SoundCue /AmbientAudioContent/SoundCues/Water/Spot_Water_Slurp_Pool_Fizzy_Inside_Loop_Cue
root=null
nodes []
properties {
 bOverrideAttenuation=true
 bOverrideConcurrency=true
 maxDistance=1050
 soundClassObject=/Game/Sounds/FortSoundClassesAndModes/Fort_Water_Ambient
 volumeMultiplier=1
 attenuationOverrides {
  bAttenuateWithLPF=true
  coneSphereFalloffDistance=750
  coneSphereRadius=300
  falloffDistance=750
  lPFFrequencyAtMax=850
  lPFRadiusMax=1000
  lPFRadiusMin=500
  nonFocusVolumeAttenuation=0.4
  nonSpatializedRadiusEnd=272
  nonSpatializedRadiusStart=750
  occlusionInterpolationTime=2
  occlusionLowPassFilterFrequency=1000
  occlusionTraceChannel=ECC_GameTraceChannel11
  stereoSpread=1597.3663
  attenuationShapeExtents=<300 90 120>
 }
 concurrencyOverrides {maxCount=2, voiceStealReleaseTime=2}
 concurrencySet [/Game/Sounds/Concurrency_Rules/Waterfall_StopFarthestOldest_Conc]
}
